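# Translucent unlit Niagara material that draws the Veiled emote's emoji sprite-sheet texture as a glitchy hologram, with optional particle-color tinting.
# Two sine-driven screen-space line bands, offset by a dynamic parameter, brighten emissive and join Atlas scanlines and wiggle-driven blocks in opacity.
# Texture alpha and a noise blend driven by the Erosion parameter shape that opacity.
# The wiggle, rotation and morph-target vertex offset branch is multiplied by zero, disabling displacement.
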Material /BRCosmetics/Animation/Game/MainPlayer/Emotes/Veiled/FX/Materials/M_Emoji_Flipbook_Veiled_Emote
graph {
 properties {
  bUsedWithNiagaraMeshParticles=true
  bUsedWithNiagaraRibbons=true
  bUsedWithNiagaraSprites=true
  blendMode=BLEND_Translucent
  pixelDepthOffsetMode=PDOM_Legacy
  shadingModel=MSM_Unlit
  usedShadingModels=Unlit
 }
 nodes [
  n1 = ParticleColor()
  n2 = TextureSampleParameter2D() {parameterName=TextureSpriteSheet, texture=/Game/Animation/Game/MainPlayer/Emotes/Veiled/FX/Textures/T_Symbol_Pattern_Veiled_Emote}
  n3 = MaterialFunctionCall(World Position=WorldPosition().XYZ) {materialFunction=&TransformToClipSpace}
  n4 = Add(A=Time(), B=DynamicParameter() {defaultValue=<rgba 0 5 1 1>, paramNames [OpacityLerp, ScanLineRandom, Erosion, Param4]}.ScanLineRandom)
  n5 = Subtract(A=@n3.Screen aligned 0-1 UV, B=ConstantBiasScale(None=Sine(None=@n4) {period=5.41}) {bias=1.3, scale=0.35})
  n6 = ScalarParameter() {defaultValue=96, parameterName=Line Multiplier A}
  n7 = StaticComponentMaskParameter(None=Clamp(None=Multiply(A=@n5, B=@n6))) {defaultG=true}
  n8 = Multiply(A=Multiply(A=@n7, B=OneMinus(None=@n7))) {constB=2.1}
  n9 = Subtract(A=@n3.Screen aligned 0-1 UV, B=ConstantBiasScale(None=Sine(None=@n4) {period=7.89}) {bias=1.25, scale=0.3})
  n10 = ScalarParameter() {defaultValue=84, parameterName=Line Multiplier B}
  n11 = StaticComponentMaskParameter(None=Clamp(None=Multiply(A=@n9, B=@n10))) {defaultG=true}
  n12 = Multiply(A=Multiply(A=@n11, B=OneMinus(None=@n11))) {constB=2.2}
  n13 = Multiply(A=Add(
    A=MaterialFunctionCall!(In=@n8) {materialFunction=/Engine/Functions/Engine_MaterialFunctions02/Utility/Round}.Result
    B=MaterialFunctionCall(In=@n12) {materialFunction=&Round}.Result
   )) {constB=0.05}
  n14 = Add(A=@n2.RGB, B=@n13)
  MaterialFunctionCall(Emissive=StaticSwitchParameter(True=Multiply(A=@n1.RGB, B=@n14), False=@n14) {parameterName=UseParticleColor}) {materialFunction=&MF_FX_EmissiveAdjustments}
  n15 = Constant3Vector() {constant=<rgba 0 0 1 0>}
  n16 = RotateAboutAxis(
   NormalizedRotationAxis=Normalize(VectorInput=Transform(None=@n15) {transformSourceType=TRANSFORMSOURCE_Local})
   RotationAngle=Multiply(A=Time(), B=ScalarParameter() {defaultValue=0.05, parameterName=RotationSpeed})
   PivotPoint=ObjectPositionWS()
   Position=WorldPosition().XYZ
  )
  n17 = DynamicParameter() {parameterIndex=1, defaultValue=<rgba 1 0 0 0>, paramNames [EdgeSpeed, BiasScale, Gradient Position, TriggerWiggle]}
  n18 = MaterialFunctionCall!(
   WPO Intensity A
   WPO Intensity B
   Time
   Bias=StaticSwitchParameter(True=@n17.TriggerWiggle, False=Constant()) {defaultValue=true, parameterName=Use Param4 for Wiggle}
  ) {materialFunction=/Game/Effects/Fort_Effects/Effects/Atlas/Materials/MF_Atlas_HologramWiggle_01}
  n19 = Multiply(A=ScalarParameter() {defaultValue=0.1, parameterName=Time Speed}, B=Time() {period=1})
  n20 = TextureObjectParameter() {
   parameterName=Morph Target Normals
   samplerType=SAMPLERTYPE_LinearColor
   texture=/Game/Global/Textures/Default/Blanks/T_BlackVectorDisplacement
  }
  n21 = TextureObjectParameter() {
   parameterName=Morph Target Offsets
   samplerType=SAMPLERTYPE_LinearColor
   texture=&T_BlackVectorDisplacement
  }
  n22 = ComponentMask(None=TextureProperty(Tex=@n20)) {g=true}
  n23 = StaticBool() {value=true}
  n24 = StaticSwitchParameter(
   True=Add(
    A=@n18.Result
    B=MaterialFunctionCall!(Morph Animation=@n19, Morph Normal=@n20, Morph Texture=@n21, Number of Morph Targets=@n22, 0-1 Animation Value?=@n23, Morph Target UV) {materialFunction=/Engine/Functions/Engine_MaterialFunctions02/MAXScripts/MS_VertexAnimationTools_MorphTargets}.World Position Offset
   )
   False=@n18.Result
  ) {parameterName=Use Morph Targets?}
  n25 = FeatureLevelSwitch(Default=@n16, ES2=@n24, ES3_1=@n24, SM4_REMOVED=@n16)
  Multiply(A=StaticSwitchParameter(True=StaticSwitchParameter(True=@n25, False=@n24) {parameterName=Use Rotation}, False=Constant4Vector().RGBA) {defaultValue=true, parameterName=Use WPO}) {constB=0}
  n26 = Constant()
  n27 = MaterialFunctionCall!(
   Number of Scanlines=ScalarParameter() {defaultValue=150, parameterName=Scanline Amount}
   Scanline Speed
   "Scanlines Thickness (Bias)"=ScalarParameter() {defaultValue=-0.15, parameterName=ScanlineSize}
   Scanlines Dark Value=ScalarParameter() {defaultValue=0.5, parameterName=Scanlines Dark}
   Scanlines Light Value=ScalarParameter() {defaultValue=0.6, parameterName=Scanlines Light}
   Block Opacity=Multiply(A=StaticSwitchParameter(True=@n18.Intense 0-1, False=@n26) {defaultValue=true, parameterName=Use Blocks?}) {constB=0.4}
   UVs
   Time
   Block Repeat
  ) {materialFunction=/Game/Effects/Fort_Effects/Effects/Atlas/Materials/MF_Atlas_Scanlines_01}
  n28 = TextureCoordinate()
  n29 = Subtract(A=TextureSample(UVs=Multiply(A=@n28)) {texture=/Game/Athena/UI/Challenges/Art/T_UI_DataChip_Noise}.G) {constB=0.5}
  n30 = DynamicParameter() {defaultValue=<rgba 0 5 1 1>, paramNames [OpacityLerp, ScanLineRandom, Erosion, Param4]}
  n31 = Multiply(A=@n2.A, B=Multiply(A=Add(A=@n13, B=@n27.Result), B=LinearInterpolate(A=@n29, Alpha=@n30.Erosion) {constAlpha=0}))
  StaticSwitchParameter(True=Multiply(A=@n1.A, B=@n31), False=@n31) {parameterName=UseParticleColor}
 ]
}
