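# Niagara effect for the Veiled character's phone emote: a blue glow with emoji sprites popping out, all at small fixed offsets.
# Two Corona emitters burst glow sprites whose alpha combines waveform flicker with fade curves.
# Two looping Emojis emitters stagger bursts of SubUV sprites, launched in an upward cone and slowed by drag.
# SubUV_ConeGlowLoop continuously spawns soft drifting glows, sharing the emojis' cyan-to-blue color curve.
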NiagaraSystem /BRCosmetics/Animation/Game/MainPlayer/Emotes/Veiled/FX/NS_Veiled_Phone_Emote
dependencies {
 usedDataInterfaces [&NiagaraDataInterfaceCurve, &NiagaraDataInterfaceColorCurve]
 usedDynamicInputs []
 usedModules [&SystemState]
}
emitters {
 Corona {
  emitterUpdateScript [
   &EmitterState()
   &SpawnBurst_Instantaneous(Spawn Count=5, Use Spawn Probability=false, UseLoopCountLimit=false)
  ]
  particleSpawnScript [
   &InitializeParticle(
    Color=<rgba 0.01 0.505 1 0.1>
    Lifetime=2.5
    Position=<12.5 -5 -2.5>
    Position Mode=::NewEnumerator1
    Sprite Size Mode=::NewEnumerator3
    Uniform Sprite Size=20
    UsePositionOffset=false
    Write Material Random=true
   )
   &DynamicMaterialParameters(
    Index 0 Param 1=2
    Module.Param0WriteEnabled=true
    Module.Param1WriteEnabled=true
    Module.Param2WriteEnabled=true
    Module.Param3WriteEnabled=true
   )
  ]
  particleUpdateScript [
   &ParticleState()
   /SetVariables(Particles.SpriteFacingFlat:Vector3f=<1 0 0>) {isSetParameters=true}
   &ScaleColor(
    Scale Alpha=&Waveform(
     Waveform Count=::NewEnumerator1
     "Waveform [1]"=::NewEnumerator0
     "[1] Frequency"=4
     "[1] Amplitude Min/Max"=<0 1>
     "[1] Scale"=0.2
     "[1] Bias"=1
     "[2] Average"=false
     "[2] Input Position"=0
     "[2] Phase"=$Emitter.Age
     "[2] Phase Scale"=1
     "[2] Compound Frequency"=2
     "[2] Frequency"=3
     "[2] Amplitude Min/Max"=<0 1>
     "[2] Scale"=0.2
     "[2] Bias"=0
     "[2] Exponent"=1
     "[2] Pulse Duration"=0.1
     "[2] Clip Triangle Min/Max"=<0 0>
     "[2] Random Seed"=0
    )
    ScaleA=true
    ScaleRGB=true
   )
   &ScaleColor(
    Scale Alpha=&FloatFromCurve(
     FloatCurve=DataInterface() {propertyValues {Curve {keys [
         {arriveTangent=1.3552, interpMode=RCIM_Cubic, leaveTangent=1.3552, tangentMode=RCTM_User, time=0, value=0}
         <at 0.2 1 RCIM_Cubic>
         <at 0.9 1 RCIM_Cubic>
         {arriveTangent=-2.0123, interpMode=RCIM_Cubic, leaveTangent=-2.0123, tangentMode=RCTM_User, time=1, value=0}
        ]}}}
     Scale Curve=0.8
    )
    ScaleA=true
    ScaleRGB=true
   ) {module=ScaleColor001}
  ]
  rendererClasses [&NiagaraSpriteRendererProperties]
  renderers [{rendererClass=&NiagaraSpriteRendererProperties}]
 }
 CoronaLOW {
  emitterUpdateScript [
   &EmitterState()
   &SpawnBurst_Instantaneous(Use Spawn Probability=false, UseLoopCountLimit=false)
  ]
  particleSpawnScript [
   &InitializeParticle(
    Color=<rgba 0.01 0.4555 1 0.3>
    Lifetime=2.5
    Position=<12.5 -5 -2.5>
    Position Mode=::NewEnumerator1
    Sprite Size Mode=::NewEnumerator3
    Uniform Sprite Size=20
    UsePositionOffset=false
    Write Material Random=true
   )
   &DynamicMaterialParameters(
    Index 0 Param 1=2
    Module.Param0WriteEnabled=true
    Module.Param1WriteEnabled=true
    Module.Param2WriteEnabled=true
    Module.Param3WriteEnabled=true
   )
  ]
  particleUpdateScript [
   &ParticleState()
   &ScaleColor(
    Scale Alpha=&Waveform(
     Waveform Count=::NewEnumerator1
     "Waveform [1]"=::NewEnumerator0
     "[1] Frequency"=4
     "[1] Amplitude Min/Max"=<0 1>
     "[1] Scale"=0.2
     "[1] Bias"=1
     "[2] Average"=false
     "[2] Input Position"=0
     "[2] Phase"=$Emitter.Age
     "[2] Phase Scale"=1
     "[2] Compound Frequency"=2
     "[2] Frequency"=3
     "[2] Amplitude Min/Max"=<0 1>
     "[2] Scale"=0.2
     "[2] Bias"=0
     "[2] Exponent"=1
     "[2] Pulse Duration"=0.1
     "[2] Clip Triangle Min/Max"=<0 0>
     "[2] Random Seed"=0
    )
    ScaleA=true
    ScaleRGB=true
   )
   &ScaleColor(
    Scale Alpha=&FloatFromCurve(FloatCurve=DataInterface() {propertyValues {Curve {keys [
         {arriveTangent=1.3552, interpMode=RCIM_Cubic, leaveTangent=1.3552, tangentMode=RCTM_User, time=0, value=0}
         <at 0.1 1 RCIM_Cubic>
         {arriveTangent=-1.1111, interpMode=RCIM_Cubic, leaveTangent=-1.1111, time=0.9, value=1}
         {arriveTangent=-2.0123, interpMode=RCIM_Cubic, leaveTangent=-2.0123, tangentMode=RCTM_User, time=1, value=0}
        ]}}})
    ScaleA=true
    ScaleRGB=true
   ) {module=ScaleColor001}
  ]
  rendererClasses [&NiagaraSpriteRendererProperties]
  renderers [{rendererClass=&NiagaraSpriteRendererProperties}]
 }
 Emojis {
  emitterUpdateScript [
   &EmitterState(Life Cycle Mode=::NewEnumerator1, Loop Behavior=::NewEnumerator1, Loop Delay=0.2, Loop Duration=1.5, UseLoopDelay=true)
   &SpawnBurst_Instantaneous(Use Spawn Probability=true, UseLoopCountLimit=false)
   &SpawnBurst_Instantaneous(Spawn Time=0.75, Use Spawn Probability=true, UseLoopCountLimit=false) {module=SpawnBurst_Instantaneous001}
  ]
  particleSpawnScript [
   &InitializeParticle(
    Lifetime=&RandomRangeFloat(Minimum=0.5, Maximum=0.5)
    Position=<12 0 0>
    Position Mode=::NewEnumerator1
    Sprite Size Mode=::NewEnumerator1
    Uniform Sprite Size Max=12.5
    Uniform Sprite Size Min=7.5
    UsePositionOffset=false
    Write Material Random=true
   )
   &AddVelocity(
    Cone Axis=<0 0 1>
    Inner Cone Angle=45
    Module.UseConeFalloff=false
    Module.UseVelDistribution=false
    Override Seed=false
    Rotation Coordinate Space=::World
    Velocity Mode=::NewEnumerator2
    Velocity Speed=&RandomRangeFloat(Minimum=450, Maximum=500)
   )
   &SubUVAnimation(End Frame=4, SubUV Animation Mode:ENiagara_SubUVLookupMode=::NewEnumerator2)
   &DynamicMaterialParameters(
    Index 1 Param 4=1
    Module.Param0WriteEnabled1=true
    Module.Param1WriteEnabled1=true
    Module.Param2WriteEnabled1=true
    Module.Param3WriteEnabled1=true
    Write Parameter Index 0=false
    Write Parameter Index 1=true
   )
  ]
  particleUpdateScript [
   &Drag(Drag=&RandomRangeFloat(Minimum=12.5, Maximum=15), Use Linear Drag=true, Use Rotational Drag=true)
   &SolveForcesAndVelocity(Clamp Velocity=false, Limit Acceleration=false)
   &ParticleState()
   &ScaleSpriteSize(Uniform Curve Sprite Scale=DataInterface() {propertyValues {Curve {keys [
        {arriveTangent=0.8117, interpMode=RCIM_Cubic, leaveTangent=0.8117, tangentMode=RCTM_User, time=0, value=0.5}
        {arriveTangent=1.2225, interpMode=RCIM_Cubic, leaveTangent=1.2225, time=0.4, value=0.94}
        {arriveTangent=0.1, interpMode=RCIM_Cubic, leaveTangent=0.1, time=0.6907, value=1.3444}
        <at 1 1 RCIM_Cubic>
       ]}}})
   &ScaleColor(
    Scale Alpha=&FloatFromCurve(FloatCurve=DataInterface() {propertyValues {Curve {keys [
         {interpMode=RCIM_Cubic, tangentMode=RCTM_User, time=0, value=0}
         {interpMode=RCIM_Cubic, tangentMode=RCTM_User, time=0.5, value=1}
         {tangentMode=RCTM_User, time=1, value=0}
        ]}}})
    Scale RGB=&MakeVectorFromLinearColorRGB(Color=&ColorFromCurve(ColorCurve=DataInterface() {propertyValues {
        AlphaCurve {keys [<at 0 1 RCIM_Cubic>,<at 1 1 RCIM_Cubic>]}
        BlueCurve {keys [<at 0.4 1 RCIM_Cubic>,<at 0.7 1>]}
        GreenCurve {keys [<at 0.4 0.788 RCIM_Cubic>,<at 0.7 0.232>]}
        RedCurve {keys [<at 0.4 0.168 RCIM_Cubic>,<at 0.7 0.155>]}
       }}))
    ScaleA=true
    ScaleRGB=true
   )
   &CameraOffset(Camera Offset Amount=15)
  ]
  rendererClasses [&NiagaraSpriteRendererProperties]
  renderers [{rendererClass=&NiagaraSpriteRendererProperties}]
 }
 Emojis001 {
  emitterUpdateScript [
   &EmitterState(Life Cycle Mode=::NewEnumerator1, Loop Behavior=::NewEnumerator1, Loop Delay=0.2, Loop Duration=1.5, UseLoopDelay=true)
   &SpawnBurst_Instantaneous(Spawn Time=0.4, Use Spawn Probability=true, UseLoopCountLimit=false)
   &SpawnBurst_Instantaneous(Spawn Time=1.2, Use Spawn Probability=true, UseLoopCountLimit=false) {module=SpawnBurst_Instantaneous001}
  ]
  particleSpawnScript [
   &InitializeParticle(
    Lifetime=&RandomRangeFloat(Minimum=0.4, Maximum=0.4)
    Position=<12 0 0>
    Position Mode=::NewEnumerator1
    Sprite Size Mode=::NewEnumerator1
    Uniform Sprite Size Max=12
    Uniform Sprite Size Min=8
    UsePositionOffset=false
    Write Material Random=true
   )
   &AddVelocity(
    Cone Axis=<0 0 1>
    Inner Cone Angle=45
    Module.UseConeFalloff=false
    Module.UseVelDistribution=false
    Override Seed=false
    Rotation Coordinate Space=::World
    Velocity Mode=::NewEnumerator2
    Velocity Speed=&RandomRangeFloat(Minimum=300, Maximum=350)
   )
   &SubUVAnimation(End Frame=4, SubUV Animation Mode:ENiagara_SubUVLookupMode=::NewEnumerator2)
   &DynamicMaterialParameters(
    Index 1 Param 4=1
    Module.Param0WriteEnabled1=true
    Module.Param1WriteEnabled1=true
    Module.Param2WriteEnabled1=true
    Module.Param3WriteEnabled1=true
    Write Parameter Index 0=false
    Write Parameter Index 1=true
   )
  ]
  particleUpdateScript [
   &Drag(Drag=&RandomRangeFloat(Minimum=12, Maximum=15), Use Linear Drag=true, Use Rotational Drag=true)
   &SolveForcesAndVelocity(Clamp Velocity=false, Limit Acceleration=false)
   &ParticleState()
   &ScaleSpriteSize(Uniform Curve Sprite Scale=DataInterface() {propertyValues {Curve {keys [
        {arriveTangent=0.8117, interpMode=RCIM_Cubic, leaveTangent=0.8117, tangentMode=RCTM_User, time=0, value=0.5}
        {arriveTangent=1.2225, interpMode=RCIM_Cubic, leaveTangent=1.2225, time=0.4, value=0.94}
        {arriveTangent=0.1, interpMode=RCIM_Cubic, leaveTangent=0.1, time=0.6907, value=1.3444}
        <at 1 1 RCIM_Cubic>
       ]}}})
   &ScaleColor(
    Scale Alpha=&FloatFromCurve(FloatCurve=DataInterface() {propertyValues {Curve {keys [
         {interpMode=RCIM_Cubic, tangentMode=RCTM_User, time=0, value=0}
         {interpMode=RCIM_Cubic, tangentMode=RCTM_User, time=0.5, value=1}
         {tangentMode=RCTM_User, time=1, value=0}
        ]}}})
    Scale RGB=&MakeVectorFromLinearColorRGB(Color=&ColorFromCurve(ColorCurve=DataInterface() {propertyValues {
        AlphaCurve {keys [<at 0 1 RCIM_Cubic>,<at 1 1 RCIM_Cubic>]}
        BlueCurve {keys [<at 0.4 1 RCIM_Cubic>,<at 0.7 1>]}
        GreenCurve {keys [<at 0.4 0.788 RCIM_Cubic>,<at 0.7 0.232>]}
        RedCurve {keys [<at 0.4 0.168 RCIM_Cubic>,<at 0.7 0.155>]}
       }}))
    ScaleA=true
    ScaleRGB=true
   )
   &CameraOffset(Camera Offset Amount=15)
  ]
  rendererClasses [&NiagaraSpriteRendererProperties]
  renderers [{rendererClass=&NiagaraSpriteRendererProperties}]
 }
 SubUV_ConeGlowLoop {
  emitterUpdateScript [
   &EmitterState(Life Cycle Mode=::NewEnumerator1, Loop Behavior=::NewEnumerator1, Loop Duration=2, UseLoopDelay=false)
   &SpawnRate(SpawnRate=10, Use Spawn Probability=false)
  ]
  particleSpawnScript [
   &InitializeParticle(
    Color=<rgba 0.9174 0.8073 1 0.5>
    Lifetime=&RandomRangeFloat(Minimum=0.75, Maximum=1)
    Position=<12 0 0>
    Position Mode=::NewEnumerator1
    Sprite Size Max=<50 35>
    Sprite Size Min=<40 30>
    Sprite Size Mode=::NewEnumerator2
    UsePositionOffset=false
    Write Material Random=true
   )
   &AddVelocity(Coordinate Space=::World, Velocity=<0 -0.25 1>)
   &SubUVAnimation(End Frame=0)
  ]
  particleUpdateScript [
   &SolveForcesAndVelocity(Clamp Velocity=false, Limit Acceleration=false)
   &ParticleState()
   &ScaleSpriteSize(Uniform Curve Sprite Scale=DataInterface() {propertyValues {Curve {keys [
        {arriveTangent=0.8117, interpMode=RCIM_Cubic, leaveTangent=0.8117, tangentMode=RCTM_User, time=0, value=0.5}
        {arriveTangent=0.7071, interpMode=RCIM_Cubic, leaveTangent=0.7071, time=0.4, value=0.94}
        {arriveTangent=0.1, interpMode=RCIM_Cubic, leaveTangent=0.1, time=0.7, value=0.995}
        <at 1 1 RCIM_Cubic>
       ]}}})
   &ScaleColor(
    Scale Alpha=&FloatFromCurve(FloatCurve=DataInterface() {propertyValues {Curve {keys [
         {interpMode=RCIM_Cubic, tangentMode=RCTM_User, time=0, value=0}
         {interpMode=RCIM_Cubic, tangentMode=RCTM_User, time=0.5, value=1}
         {tangentMode=RCTM_User, time=1, value=0}
        ]}}})
    Scale RGB=&MakeVectorFromLinearColorRGB(Color=&ColorFromCurve(ColorCurve=DataInterface() {propertyValues {
        AlphaCurve {keys [<at 0 1 RCIM_Cubic>,<at 1 1 RCIM_Cubic>]}
        BlueCurve {keys [<at 0.4 1 RCIM_Cubic>,<at 0.7 1>]}
        GreenCurve {keys [<at 0.4 0.788 RCIM_Cubic>,<at 0.7 0.232>]}
        RedCurve {keys [<at 0.4 0.168 RCIM_Cubic>,<at 0.7 0.155>]}
       }}))
    ScaleA=true
    ScaleRGB=true
   )
   &CameraOffset(Camera Offset Amount=5)
  ]
  rendererClasses [&NiagaraSpriteRendererProperties]
  renderers [{rendererClass=&NiagaraSpriteRendererProperties}]
 }
}
summary {
 systemName=NS_Veiled_Phone_Emote
 emitters [
  {emitterName=Corona, rendererClasses [&NiagaraSpriteRendererProperties]}
  {emitterName=CoronaLOW, rendererClasses [&NiagaraSpriteRendererProperties]}
  {emitterName=SubUV_ConeGlowLoop, rendererClasses [&NiagaraSpriteRendererProperties]}
  {emitterName=Emojis, rendererClasses [&NiagaraSpriteRendererProperties]}
  {emitterName=Emojis001, rendererClasses [&NiagaraSpriteRendererProperties]}
 ]
 userVariables [User.SkeletalMesh {
   type=NiagaraDataInterfaceSkeletalMesh
   defaultValue {
    struct=NiagaraExt_VariableValue_DataInterface
    value {dataInterface=&self:NiagaraDataInterfaceSkeletalMesh_0}
   }
  }]
}
