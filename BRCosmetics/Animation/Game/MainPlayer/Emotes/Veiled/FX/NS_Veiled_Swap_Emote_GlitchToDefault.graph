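# Niagara effect for the Veiled swap emote's glitch back to the default look, a short burst whose emitter loop lengths follow User.WipeDuration.
# Cube_Glitch emitters burst sprites and meshes in boxes with curl noise, pulsing size curves, distance-based scaling and Z-distance color lerps.
# Mesh and sprite Wipe layers plus a Glow cover the body.
# Fresnel glows track skeletal mesh bones, tinted by time of day, and glitch decals follow both feet.
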NiagaraSystem /BRCosmetics/Animation/Game/MainPlayer/Emotes/Veiled/FX/NS_Veiled_Swap_Emote_GlitchToDefault
dependencies {
 usedDataInterfaces [&NiagaraDataInterfaceColorCurve, &NiagaraDataInterfaceCurve]
 usedDynamicInputs [
  &self:ScratchDynamicInput_02
  &FNTimeOfDayScalar
  &self:ScratchDynamicInput_01
 ]
 usedModules [
  &self:SystemSpawnScript.SystemScriptSource.SystemScriptGraph.NiagaraNodeAssignment_1.SetVariables_#1
  &SystemState
  &self:SystemSpawnScript.SystemScriptSource.SystemScriptGraph.NiagaraNodeAssignment_2.SetVariables_#2
  &self:SystemSpawnScript.SystemScriptSource.SystemScriptGraph.NiagaraNodeAssignment_0.SetVariables_#3
  &self:SystemSpawnScript.SystemScriptSource.SystemScriptGraph.NiagaraNodeAssignment_3.SetVariables_#4
 ]
}
emitters {
 Cube_Glitch_Head {
  emitterUpdateScript [
   &EmitterState(Life Cycle Mode=::NewEnumerator1, Loop Behavior=::NewEnumerator1, Loop Delay=0.15, Loop Duration=$User.WipeDuration, UseLoopDelay=true)
   &SpawnBurst_Instantaneous(Spawn Count=5, Use Spawn Probability=false, UseLoopCountLimit=false)
  ]
  particleSpawnScript [
   &InitializeParticle(
    Lifetime Max=1
    Lifetime Min=0.25
    Lifetime Mode=::NewEnumerator1
    Mesh Scale Mode=::NewEnumerator3
    Mesh Uniform Scale=50
    Sprite Size Max=<60 60>
    Sprite Size Min=<25 50>
    Sprite Size Mode=::NewEnumerator2
    UsePositionOffset=true
    Write Material Random=true
   ) {module=InitializeParticle001}
   &DynamicMaterialParameters(
    Index 0 Param 4=&RandomRangeFloat(Minimum=-1, Maximum=0.1)
    Index 1 Param 1=&RandomRangeFloat(Minimum=-100, Maximum=100)
    Module.Param0WriteEnabled=true
    Module.Param0WriteEnabled1=true
    Module.Param1WriteEnabled=true
    Module.Param1WriteEnabled1=true
    Module.Param2WriteEnabled=true
    Module.Param2WriteEnabled1=true
    Module.Param3WriteEnabled=true
    Module.Param3WriteEnabled1=true
    Write Parameter Index 1=true
   ) {module=DynamicMaterialParameters001}
   &ShapeLocation(
    Box Midpoint=<0.5 0.5 0>
    Box Size=<30 15 160>
    MaskSpawnGroup=false
    Override Seed=false
    Shape Primitive=::NewEnumerator2
    Surface Only=false
    UseOwnerScale=true
   )
  ]
  particleUpdateScript [
   &ParticleState()
   &ScaleSpriteSize(
    Scale Sprite Size Mode=::NewEnumerator0
    Uniform Scale Factor=&FloatFromCurve(
     FloatCurve=DataInterface() {propertyValues {Curve {keys [
         {arriveTangent=5.4245, interpMode=RCIM_Cubic, leaveTangent=5.4245, tangentMode=RCTM_User, time=0, value=5}
         {arriveTangent=-3.9614, interpMode=RCIM_Cubic, leaveTangent=-3.9614, time=0.0963, value=8.5355}
         {arriveTangent=-22.2662, interpMode=RCIM_Cubic, leaveTangent=-22.2662, time=0.2702, value=3.9295}
         <at 0.3 4>
         <at 0.4 11.5>
         {arriveTangent=0.1532, interpMode=RCIM_Cubic, leaveTangent=0.1532, tangentMode=RCTM_User, time=0.5, value=2.875}
         <at 0.6 3>
         <at 0.7 11.5>
         <at 0.8 2.5>
         <at 0.9 2>
         {arriveTangent=-0.0405, interpMode=RCIM_Cubic, leaveTangent=-0.0405, tangentMode=RCTM_User, time=1, value=0.5}
        ]}}}
     Scale Curve=&Multiply_Float(
      A=0.1
      B=&ClampFloat(
       Float=&FN_LOD_Distance_Remap(Start Distance=200, End Distance=400, Minimum Output=0.5, Maximum Output=1)
       Min=0.5
      )
     )
    )
   )
   /SetVariables(Particles.ZDistance:NiagaraFloat=&self:ScratchDynamicInput_03(Distance:NiagaraFloat=150, SK_Mesh_Pos:NiagaraPosition=$Engine.Owner.Position)) {isSetParameters=true, module=SetVariables_#5}
   &ScaleMeshSize(Scale Factor=&Multiply_VectorByFloat(
     Vector=<1 1 1>
     Float=&Multiply_Float(
      A=0.08
      B=&FloatFromCurve(
       FloatCurve=DataInterface() {propertyValues {Curve {keys [
           {arriveTangent=5.4245, interpMode=RCIM_Cubic, leaveTangent=5.4245, tangentMode=RCTM_User, time=0, value=5}
           {arriveTangent=-3.9614, interpMode=RCIM_Cubic, leaveTangent=-3.9614, time=0.0963, value=8.5355}
           {arriveTangent=-22.2662, interpMode=RCIM_Cubic, leaveTangent=-22.2662, time=0.2702, value=3.9295}
           <at 0.3 4>
           <at 0.4 11.5>
           {arriveTangent=0.1532, interpMode=RCIM_Cubic, leaveTangent=0.1532, tangentMode=RCTM_User, time=0.5, value=2.875}
           <at 0.6 3>
           <at 0.7 11.5>
           <at 0.8 2.5>
           <at 0.9 2>
           {arriveTangent=-0.0405, interpMode=RCIM_Cubic, leaveTangent=-0.0405, tangentMode=RCTM_User, time=1, value=0.5}
          ]}}}
       Scale Curve=&Multiply_Float(
        A=0.6
        B=&ClampFloat(
         Float=&FN_LOD_Distance_Remap(Start Distance=200, End Distance=400, Minimum Output=0.5, Maximum Output=1)
         Min=0.5
        )
       )
      )
     )
    ))
   &ScaleColor(
    Scale Mode=::NewEnumerator1
    Scale RGBA=&MakeVector4_FromLinearColor(LinearColor=&LerpLinearColors(StartColor=<rgba 0.9219 0 0.2972 1>, EndColor=<rgba 0 0.803 1 1>, LerpFactor=$Particles.ZDistance))
    ScaleRGBA=true
   )
   &CameraOffset(Camera Offset Amount=&RandomRangeFloat(Minimum=-40, Maximum=10))
   &Drag(Drag=10, Use Linear Drag=true, Use Rotational Drag=true)
   &CurlNoiseForce(Noise Frequency=100, Noise Strength=250, Pan Noise=false)
   &SolveForcesAndVelocity(Clamp Velocity=false, Limit Acceleration=false)
   &DynamicMaterialParameters(
    Index 0 Param 2=&RandomRangeFloat(Minimum=-1, Maximum=1)
    Index 0 Param 3=&FloatFromCurve(FloatCurve=DataInterface() {propertyValues {Curve {keys [
         <at 0 2 RCIM_Cubic>
         {arriveTangent=-6.2285, interpMode=RCIM_Cubic, leaveTangent=-6.2285, tangentMode=RCTM_User, time=0.5612, value=1.0841}
         {arriveTangent=-0.8021, interpMode=RCIM_Cubic, leaveTangent=-0.8021, tangentMode=RCTM_User, time=1, value=0}
        ]}}})
    Index 1 Param 4=&FloatFromCurve(FloatCurve=DataInterface() {propertyValues {Curve {keys [
         {arriveTangent=4.1785, interpMode=RCIM_Cubic, leaveTangent=4.1785, tangentMode=RCTM_User, time=0, value=0}
         {arriveTangent=0.9995, interpMode=RCIM_Cubic, leaveTangent=0.9995, tangentMode=RCTM_User, time=0.2223, value=0.6825}
         {arriveTangent=-0.0405, interpMode=RCIM_Cubic, leaveTangent=-0.0405, tangentMode=RCTM_User, time=1, value=1}
        ]}}})
    Module.Param0WriteEnabled=false
    Module.Param0WriteEnabled1=false
    Module.Param1WriteEnabled=true
    Module.Param1WriteEnabled1=false
    Module.Param2WriteEnabled=true
    Module.Param2WriteEnabled1=false
    Module.Param3WriteEnabled=true
    Module.Param3WriteEnabled1=true
    Write Parameter Index 1=true
   ) {module=DynamicMaterialParameters002}
  ]
  rendererClasses [&NiagaraSpriteRendererProperties]
  renderers [{rendererClass=&NiagaraSpriteRendererProperties}]
 }
 Cube_Glitch_Head001 {
  emitterUpdateScript [
   &EmitterState(Life Cycle Mode=::NewEnumerator1, Loop Behavior=::NewEnumerator1, Loop Delay=0.15, Loop Duration=$User.WipeDuration, UseLoopDelay=true)
   &SpawnBurst_Instantaneous(Spawn Count=10, Use Spawn Probability=false, UseLoopCountLimit=false)
   &SpawnBurst_Instantaneous(Spawn Count=10, Spawn Time=0.5, Use Spawn Probability=false, UseLoopCountLimit=false) {module=SpawnBurst_Instantaneous001}
  ]
  particleSpawnScript [
   &InitializeParticle(
    Lifetime Max=1
    Lifetime Min=0.25
    Lifetime Mode=::NewEnumerator1
    Mesh Scale Max=<1.5 1.5 2>
    Mesh Scale Min=<0.5 0.5 1>
    Mesh Scale Mode=::NewEnumerator2
    UsePositionOffset=true
    Write Material Random=true
   ) {module=InitializeParticle001}
   &DynamicMaterialParameters(
    Index 0 Param 4=&RandomRangeFloat(Minimum=-1, Maximum=0.1)
    Index 1 Param 1=&RandomRangeFloat(Minimum=-100, Maximum=100)
    Module.Param0WriteEnabled=true
    Module.Param0WriteEnabled1=true
    Module.Param1WriteEnabled=true
    Module.Param1WriteEnabled1=true
    Module.Param2WriteEnabled=true
    Module.Param2WriteEnabled1=true
    Module.Param3WriteEnabled=true
    Module.Param3WriteEnabled1=true
    Write Parameter Index 1=true
   ) {module=DynamicMaterialParameters001}
   &ShapeLocation(
    Box Midpoint=<0.5 0.5 0>
    Box Size=<100 100 200>
    MaskSpawnGroup=false
    Override Seed=false
    Shape Primitive=::NewEnumerator2
    Surface Only=false
    UseOwnerScale=true
   )
  ]
  particleUpdateScript [
   &ParticleState()
   /SetVariables(Particles.ZDistance:NiagaraFloat=&self:ScratchDynamicInput_03(Distance:NiagaraFloat=150, SK_Mesh_Pos:NiagaraPosition=$Engine.Owner.Position)) {isSetParameters=true, module=SetVariables_#5}
   &ScaleMeshSize(Scale Factor=&Multiply_VectorByFloat(
     Vector=<1 1 1>
     Float=&Multiply_Float(
      A=0.08
      B=&FloatFromCurve(
       FloatCurve=DataInterface() {propertyValues {Curve {keys [
           {arriveTangent=5.4245, interpMode=RCIM_Cubic, leaveTangent=5.4245, tangentMode=RCTM_User, time=0, value=5}
           {arriveTangent=-3.9614, interpMode=RCIM_Cubic, leaveTangent=-3.9614, time=0.0963, value=8.5355}
           {arriveTangent=-22.2662, interpMode=RCIM_Cubic, leaveTangent=-22.2662, time=0.2702, value=3.9295}
           <at 0.3 4>
           <at 0.4 11.5>
           {arriveTangent=0.1532, interpMode=RCIM_Cubic, leaveTangent=0.1532, tangentMode=RCTM_User, time=0.5, value=2.875}
           <at 0.6 3>
           <at 0.7 11.5>
           <at 0.8 2.5>
           <at 0.9 2>
           {arriveTangent=-0.0405, interpMode=RCIM_Cubic, leaveTangent=-0.0405, tangentMode=RCTM_User, time=1, value=0.5}
          ]}}}
       Scale Curve=&Multiply_Float(
        A=0.6
        B=&ClampFloat(
         Float=&FN_LOD_Distance_Remap(Start Distance=200, End Distance=400, Minimum Output=0.5, Maximum Output=1)
         Min=0.5
        )
       )
      )
     )
    ))
   &ScaleColor(
    Scale Mode=::NewEnumerator1
    Scale RGBA=&MakeVector4_FromLinearColor(LinearColor=&LerpLinearColors(StartColor=<rgba 0.9219 0 0.2972 1>, EndColor=<rgba 0 0.803 1 1>, LerpFactor=$Particles.ZDistance))
    ScaleRGBA=true
   )
   &CameraOffset(Camera Offset Amount=10)
   &Drag(Drag=10, Use Linear Drag=true, Use Rotational Drag=true)
   &CurlNoiseForce(Noise Frequency=100, Noise Strength=250, Pan Noise=false)
   &SolveForcesAndVelocity(Clamp Velocity=false, Limit Acceleration=false)
   &DynamicMaterialParameters(
    Index 0 Param 2=&RandomRangeFloat(Minimum=-1, Maximum=1)
    Index 0 Param 3=&FloatFromCurve(FloatCurve=DataInterface() {propertyValues {Curve {keys [
         <at 0 2 RCIM_Cubic>
         {arriveTangent=-6.2285, interpMode=RCIM_Cubic, leaveTangent=-6.2285, tangentMode=RCTM_User, time=0.5612, value=1.0841}
         {arriveTangent=-0.8021, interpMode=RCIM_Cubic, leaveTangent=-0.8021, tangentMode=RCTM_User, time=1, value=0}
        ]}}})
    Index 1 Param 4=&FloatFromCurve(FloatCurve=DataInterface() {propertyValues {Curve {keys [
         {arriveTangent=4.1785, interpMode=RCIM_Cubic, leaveTangent=4.1785, tangentMode=RCTM_User, time=0, value=0}
         {arriveTangent=0.9995, interpMode=RCIM_Cubic, leaveTangent=0.9995, tangentMode=RCTM_User, time=0.2223, value=0.6825}
         {arriveTangent=-0.0405, interpMode=RCIM_Cubic, leaveTangent=-0.0405, tangentMode=RCTM_User, time=1, value=1}
        ]}}})
    Module.Param0WriteEnabled=false
    Module.Param0WriteEnabled1=false
    Module.Param1WriteEnabled=true
    Module.Param1WriteEnabled1=false
    Module.Param2WriteEnabled=true
    Module.Param2WriteEnabled1=false
    Module.Param3WriteEnabled=true
    Module.Param3WriteEnabled1=true
    Write Parameter Index 1=true
   ) {module=DynamicMaterialParameters002}
  ]
  rendererClasses [&NiagaraMeshRendererProperties]
  renderers [{rendererClass=&NiagaraMeshRendererProperties, rendererIndex=0}]
 }
 Cube_Glitch_Head002 {
  emitterUpdateScript [
   &EmitterState(Life Cycle Mode=::NewEnumerator1, Loop Behavior=::NewEnumerator1, Loop Delay=0.15, Loop Duration=$User.WipeDuration, UseLoopDelay=true)
   &SpawnBurst_Instantaneous(Spawn Count=3, Use Spawn Probability=false, UseLoopCountLimit=false)
   &SpawnBurst_Instantaneous(Spawn Count=3, Spawn Time=0.2, Use Spawn Probability=false, UseLoopCountLimit=false) {module=SpawnBurst_Instantaneous001}
   &SpawnBurst_Instantaneous(Spawn Count=10, Spawn Time=0.5, Use Spawn Probability=false, UseLoopCountLimit=false) {module=SpawnBurst_Instantaneous002}
  ]
  particleSpawnScript [
   &InitializeParticle(
    Lifetime Max=0.5
    Lifetime Min=0.25
    Lifetime Mode=::NewEnumerator1
    Mesh Scale Max=<0.5 5 1>
    Mesh Scale Min=<0.25 2.5 0.25>
    Mesh Scale Mode=::NewEnumerator2
    UsePositionOffset=true
    Write Material Random=true
   ) {module=InitializeParticle001}
   &DynamicMaterialParameters(
    Index 0 Param 4=&RandomRangeFloat(Minimum=-1, Maximum=0.1)
    Index 1 Param 1=&RandomRangeFloat(Minimum=-100, Maximum=100)
    Module.Param0WriteEnabled=true
    Module.Param0WriteEnabled1=true
    Module.Param1WriteEnabled=true
    Module.Param1WriteEnabled1=true
    Module.Param2WriteEnabled=true
    Module.Param2WriteEnabled1=true
    Module.Param3WriteEnabled=true
    Module.Param3WriteEnabled1=true
    Write Parameter Index 1=true
   ) {module=DynamicMaterialParameters001}
   &ShapeLocation(
    Box Midpoint=<0.5 0.5 0>
    Box Size=<100 100 200>
    MaskSpawnGroup=false
    Override Seed=false
    Shape Primitive=::NewEnumerator2
    Surface Only=false
    UseOwnerScale=true
   )
  ]
  particleUpdateScript [
   &ParticleState()
   /SetVariables(Particles.ZDistance:NiagaraFloat=&self:ScratchDynamicInput_03(Distance:NiagaraFloat=150, SK_Mesh_Pos:NiagaraPosition=$Engine.Owner.Position)) {isSetParameters=true, module=SetVariables_#5}
   &ScaleMeshSize(Scale Factor=&Multiply_VectorByFloat(
     Vector=<1 1 1>
     Float=&Multiply_Float(
      A=0.08
      B=&FloatFromCurve(
       FloatCurve=DataInterface() {propertyValues {Curve {keys [
           {arriveTangent=5.4245, interpMode=RCIM_Cubic, leaveTangent=5.4245, tangentMode=RCTM_User, time=0, value=5}
           {arriveTangent=-3.9614, interpMode=RCIM_Cubic, leaveTangent=-3.9614, time=0.0963, value=8.5355}
           {arriveTangent=-22.2662, interpMode=RCIM_Cubic, leaveTangent=-22.2662, time=0.2702, value=3.9295}
           <at 0.3 4>
           <at 0.4 11.5>
           {arriveTangent=0.1532, interpMode=RCIM_Cubic, leaveTangent=0.1532, tangentMode=RCTM_User, time=0.5, value=2.875}
           <at 0.6 3>
           <at 0.7 11.5>
           <at 0.8 2.5>
           <at 0.9 2>
           {arriveTangent=-0.0405, interpMode=RCIM_Cubic, leaveTangent=-0.0405, tangentMode=RCTM_User, time=1, value=0.5}
          ]}}}
       Scale Curve=&Multiply_Float(
        A=0.6
        B=&ClampFloat(
         Float=&FN_LOD_Distance_Remap(Start Distance=200, End Distance=400, Minimum Output=0.5, Maximum Output=1)
         Min=0.5
        )
       )
      )
     )
    )) {module=ScaleMeshSize001}
   &ScaleColor(
    Scale Mode=::NewEnumerator1
    Scale RGBA=&MakeVector4_FromLinearColor(LinearColor=&LerpLinearColors(StartColor=<rgba 0.0156 1 0.4372 1>, EndColor=<rgba 0 1 0.0456 1>, LerpFactor=$Particles.ZDistance))
    ScaleRGBA=true
   )
   &CameraOffset(Camera Offset Amount=10)
   &Drag(Drag=10, Use Linear Drag=true, Use Rotational Drag=true)
   &CurlNoiseForce(Noise Frequency=100, Noise Strength=250, Pan Noise=false)
   &SolveForcesAndVelocity(Clamp Velocity=false, Limit Acceleration=false)
   &DynamicMaterialParameters(
    Index 0 Param 2=&RandomRangeFloat(Minimum=-1, Maximum=1)
    Index 0 Param 3=&FloatFromCurve(FloatCurve=DataInterface() {propertyValues {Curve {keys [
         <at 0 2 RCIM_Cubic>
         {arriveTangent=-6.2285, interpMode=RCIM_Cubic, leaveTangent=-6.2285, tangentMode=RCTM_User, time=0.5612, value=1.0841}
         {arriveTangent=-0.8021, interpMode=RCIM_Cubic, leaveTangent=-0.8021, tangentMode=RCTM_User, time=1, value=0}
        ]}}})
    Index 1 Param 4=&FloatFromCurve(FloatCurve=DataInterface() {propertyValues {Curve {keys [
         {arriveTangent=4.1785, interpMode=RCIM_Cubic, leaveTangent=4.1785, tangentMode=RCTM_User, time=0, value=0}
         {arriveTangent=0.9995, interpMode=RCIM_Cubic, leaveTangent=0.9995, tangentMode=RCTM_User, time=0.2223, value=0.6825}
         {arriveTangent=-0.0405, interpMode=RCIM_Cubic, leaveTangent=-0.0405, tangentMode=RCTM_User, time=1, value=1}
        ]}}})
    Module.Param0WriteEnabled=false
    Module.Param0WriteEnabled1=false
    Module.Param1WriteEnabled=true
    Module.Param1WriteEnabled1=false
    Module.Param2WriteEnabled=true
    Module.Param2WriteEnabled1=false
    Module.Param3WriteEnabled=true
    Module.Param3WriteEnabled1=true
    Write Parameter Index 1=true
   ) {module=DynamicMaterialParameters002}
  ]
  rendererClasses [&NiagaraMeshRendererProperties]
  renderers [{rendererClass=&NiagaraMeshRendererProperties, rendererIndex=0}]
 }
 Cube_Glitch_Head003 {
  emitterUpdateScript [
   &EmitterState(Life Cycle Mode=::NewEnumerator1, Loop Behavior=::NewEnumerator1, Loop Delay=0.09, Loop Duration=$User.WipeDuration, UseLoopDelay=true)
   &SpawnBurst_Instantaneous(Spawn Count=7, Spawn Time=0.3, Use Spawn Probability=false, UseLoopCountLimit=false) {module=SpawnBurst_Instantaneous001}
  ]
  particleSpawnScript [
   &InitializeParticle(
    Color Mode=::NewEnumerator2
    Lifetime Max=0.75
    Lifetime Min=0.5
    Lifetime Mode=::NewEnumerator1
    Mesh Scale Mode=::NewEnumerator3
    Mesh Uniform Scale=50
    Sprite Size Max=<50 60>
    Sprite Size Min=<25 50>
    Sprite Size Mode=::NewEnumerator2
    UsePositionOffset=true
    Write Material Random=true
   ) {module=InitializeParticle001}
   &FN_SetSpawnGroup(Number of Spawn Groups=7)
   &DynamicMaterialParameters(
    Index 0 Param 4=&RandomRangeFloat(Minimum=-1, Maximum=0.1)
    Index 1 Param 1=&RandomRangeFloat(Minimum=-100, Maximum=100)
    Module.Param0WriteEnabled=true
    Module.Param0WriteEnabled1=true
    Module.Param1WriteEnabled=true
    Module.Param1WriteEnabled1=true
    Module.Param2WriteEnabled=true
    Module.Param2WriteEnabled1=true
    Module.Param3WriteEnabled=true
    Module.Param3WriteEnabled1=true
    Write Parameter Index 1=true
   ) {module=DynamicMaterialParameters001}
   &ShapeLocation(
    Box Midpoint=<0.5 0.5 0>
    Box Size=<100 50 170>
    MaskSpawnGroup=false
    Override Seed=false
    Shape Primitive=::NewEnumerator2
    Surface Only=false
    UseOwnerScale=true
   )
  ]
  particleUpdateScript [
   &ParticleState()
   &ScaleSpriteSize(
    Scale Sprite Size Mode=::NewEnumerator0
    Uniform Scale Factor=&FloatFromCurve(
     FloatCurve=DataInterface() {propertyValues {Curve {keys [
         {arriveTangent=5.4245, interpMode=RCIM_Cubic, leaveTangent=5.4245, tangentMode=RCTM_User, time=0, value=5}
         {arriveTangent=-3.9614, interpMode=RCIM_Cubic, leaveTangent=-3.9614, time=0.0963, value=8.5355}
         {arriveTangent=-22.2662, interpMode=RCIM_Cubic, leaveTangent=-22.2662, time=0.2702, value=3.9295}
         <at 0.3 4>
         <at 0.4 11.5>
         {arriveTangent=0.1532, interpMode=RCIM_Cubic, leaveTangent=0.1532, tangentMode=RCTM_User, time=0.5, value=2.875}
         <at 0.6 3>
         <at 0.7 11.5>
         <at 0.8 2.5>
         <at 0.9 2>
         {arriveTangent=-0.0405, interpMode=RCIM_Cubic, leaveTangent=-0.0405, tangentMode=RCTM_User, time=1, value=0.5}
        ]}}}
     Scale Curve=&Multiply_Float(
      A=0.1
      B=&ClampFloat(
       Float=&FN_LOD_Distance_Remap(Start Distance=200, End Distance=400, Minimum Output=0.5, Maximum Output=1)
       Min=0.5
      )
     )
    )
   )
   /SetVariables(Particles.ZDistance:NiagaraFloat=&self:ScratchDynamicInput_03(Distance:NiagaraFloat=150, SK_Mesh_Pos:NiagaraPosition=$Engine.Owner.Position)) {isSetParameters=true, module=SetVariables_#5}
   &ScaleMeshSize(Scale Factor=&Multiply_VectorByFloat(
     Vector=<1 1 1>
     Float=&Multiply_Float(
      A=0.08
      B=&FloatFromCurve(
       FloatCurve=DataInterface() {propertyValues {Curve {keys [
           {arriveTangent=5.4245, interpMode=RCIM_Cubic, leaveTangent=5.4245, tangentMode=RCTM_User, time=0, value=4}
           {arriveTangent=-0.261, interpMode=RCIM_Cubic, leaveTangent=-0.261, time=0.1, value=5}
           {arriveTangent=-5, interpMode=RCIM_Cubic, leaveTangent=-5, time=0.2702, value=3.9295}
           <at 0.3 4>
           <at 0.4 7.5>
           {arriveTangent=0.1532, interpMode=RCIM_Cubic, leaveTangent=0.1532, tangentMode=RCTM_User, time=0.5, value=2.875}
           <at 0.6 8>
           <at 0.6 3>
           <at 0.8 2.5>
           <at 0.9 2>
           {arriveTangent=-0.0405, interpMode=RCIM_Cubic, leaveTangent=-0.0405, tangentMode=RCTM_User, time=1, value=0.5}
          ]}}}
       Scale Curve=&Multiply_Float(
        A=0.6
        B=&ClampFloat(
         Float=&FN_LOD_Distance_Remap(Start Distance=200, End Distance=400, Minimum Output=0.5, Maximum Output=1)
         Min=0.5
        )
       )
      )
     )
    ))
   &ScaleColor(
    Scale Mode=::NewEnumerator1
    Scale RGBA=&MakeVector4_FromLinearColor(LinearColor=&LerpLinearColors(StartColor=<rgba 0.9219 0 0.2972 1>, EndColor=<rgba 0 0.803 1 1>, LerpFactor=$Particles.ZDistance))
    ScaleRGBA=true
   )
   &CameraOffset(Camera Offset Amount=&RandomRangeFloat(Minimum=-50, Maximum=25))
   &Drag(Drag=10, Use Linear Drag=true, Use Rotational Drag=true)
   &CurlNoiseForce(Noise Frequency=100, Noise Strength=250, Pan Noise=false)
   &SolveForcesAndVelocity(Clamp Velocity=false, Limit Acceleration=false)
   &DynamicMaterialParameters(
    Index 0 Param 2=&RandomRangeFloat(Minimum=-1, Maximum=1)
    Index 0 Param 3=&FloatFromCurve(FloatCurve=DataInterface() {propertyValues {Curve {keys [
         <at 0 1 RCIM_Cubic>
         {arriveTangent=-3.1701, interpMode=RCIM_Cubic, leaveTangent=-3.1701, tangentMode=RCTM_User, time=0.5612, value=0.5}
         {arriveTangent=-0.8021, interpMode=RCIM_Cubic, leaveTangent=-0.8021, tangentMode=RCTM_User, time=1, value=0}
        ]}}})
    Index 1 Param 4=&FloatFromCurve(FloatCurve=DataInterface() {propertyValues {Curve {keys [
         {arriveTangent=4.1785, interpMode=RCIM_Cubic, leaveTangent=4.1785, tangentMode=RCTM_User, time=0, value=0}
         {arriveTangent=0.9995, interpMode=RCIM_Cubic, leaveTangent=0.9995, tangentMode=RCTM_User, time=0.2223, value=0.6825}
         {arriveTangent=-0.0405, interpMode=RCIM_Cubic, leaveTangent=-0.0405, tangentMode=RCTM_User, time=1, value=1}
        ]}}})
    Module.Param0WriteEnabled=false
    Module.Param0WriteEnabled1=false
    Module.Param1WriteEnabled=true
    Module.Param1WriteEnabled1=false
    Module.Param2WriteEnabled=true
    Module.Param2WriteEnabled1=false
    Module.Param3WriteEnabled=true
    Module.Param3WriteEnabled1=true
    Write Parameter Index 1=true
   ) {module=DynamicMaterialParameters002}
  ]
  rendererClasses [&NiagaraSpriteRendererProperties]
  renderers [{rendererClass=&NiagaraSpriteRendererProperties}]
 }
 Fresnel_Glow_Swipe {
  emitterUpdateScript [
   &EmitterState(Life Cycle Mode=::NewEnumerator1, Loop Behavior=::NewEnumerator1, Loop Delay=0.3, Loop Duration=$User.WipeDuration, UseLoopDelay=true)
   &SpawnBurst_Instantaneous(Spawn Count=23, Use Spawn Probability=false) {module=SpawnBurst_Instantaneous001}
  ]
  particleSpawnScript [
   &FN_SetSpawnGroup(Number of Spawn Groups=23) {module=FN_SetSpawnGroup001}
   &InitializeParticle(
    Color=<rgba 1 1 1 0.5>
    Lifetime=$User.WipeDuration
    Sprite Size Mode=::NewEnumerator3
    Uniform Sprite Size=35
    UsePositionOffset=false
    Write Material Random=true
   ) {module=InitializeParticle001}
  ]
  particleUpdateScript [
   &SkeletalMeshLocation(
    Bone / Socket Index=$Particles.SpawnGroup
    Bone Sampling Mode=::NewEnumerator4
    MaskSpawnGroup=false
    Skeletal Mesh=DataInterface() {propertyValues {
      DefaultMesh=null
      ExcludeBoneName=root
      OptionalComponentClass=null
      PreviewMesh=/Game/Characters/Player/Female/Medium/Bodies/F_MED_Veiled/Meshes/F_MED_Veiled
      SkinningMode=SkinOnTheFly
      SoftSourceActor=null
      SourceMode=Default
      UvSetIndex=0
      WholeMeshLOD=-1
      bExcludeBone=false
      bReadDeformedGeometry=true
      bRequireCurrentFrameData=true
      ComponentTags []
      FilteredBones [
       head
       spine_05
       spine_03
       spine_01
       clavicle_l
       clavicle_r
       upperarm_l
       upperarm_r
       upperarm_twist_01_l
       upperarm_twist_01_r
       lowerarm_l
       lowerarm_r
       lowerarm_twist_01_l
       lowerarm_twist_01_r
       hand_l
       hand_r
       pelvis
       thigh_l
       thigh_r
       thigh_twist_01_l
       thigh_twist_01_r
       calf_l
       calf_r
      ]
      FilteredSockets []
      MeshUserParameter {parameter {name=None, typeDefHandle {registeredTypeIndex=1}, varData []}}
      SamplingRegions []
     }}
   )
   &ParticleState()
   &CameraOffset(Camera Offset Amount=50)
   &DynamicMaterialParameters(
    Index 0 Param 1=&FloatFromCurve(FloatCurve=DataInterface() {propertyValues {Curve {keys [
         {arriveTangent=12.1431, interpMode=RCIM_Cubic, leaveTangent=12.1431, tangentMode=RCTM_User, time=0, value=0}
         {arriveTangent=8.6382, interpMode=RCIM_Cubic, leaveTangent=8.6382, tangentMode=RCTM_User, time=0.2, value=3}
         {arriveTangent=-0.0405, interpMode=RCIM_Cubic, leaveTangent=-0.0405, tangentMode=RCTM_User, time=1, value=5}
        ]}}})
    Index 0 Param 2=1
    Index 0 Param 3=&FloatFromCurve(FloatCurve=DataInterface() {propertyValues {Curve {keys [{arriveTangent=3.8194, interpMode=RCIM_Cubic, leaveTangent=3.8194, tangentMode=RCTM_User, time=0, value=-0.3}, <at 0.1 0>, <at 0.6 0.4>]}}})
    Module.Param0WriteEnabled=true
    Module.Param1WriteEnabled=true
    Module.Param2WriteEnabled=true
    Module.Param3WriteEnabled=true
   )
   &ScaleColor(
    Scale Alpha=&FloatFromCurve(
     FloatCurve=DataInterface() {propertyValues {Curve {keys [
         <at 0 0 RCIM_Cubic>
         {arriveTangent=1.9398, interpMode=RCIM_Cubic, leaveTangent=1.9398, tangentMode=RCTM_User, time=0.0287, value=0.3191}
         <at 0.0579 1.0382>
         {arriveTangent=-3.5366, interpMode=RCIM_Cubic, leaveTangent=-3.5366, tangentMode=RCTM_User, time=0.1547, value=1.05}
         <at 1 0>
        ]}}}
     Scale Curve=0.5
    )
    Scale RGB=&MakeVectorFromLinearColorRGB(Color=&ColorFromCurve(ColorCurve=DataInterface() {propertyValues {
        AlphaCurve {keys [<at 0 1 RCIM_Cubic>]}
        BlueCurve {keys [
          <at 0 5 RCIM_Cubic>
          {arriveTangent=-6.1442, interpMode=RCIM_Cubic, leaveTangent=-6.1442, time=0.3777, value=5.3629}
          {arriveTangent=-11.6613, interpMode=RCIM_Cubic, leaveTangent=-11.6613, time=0.6386, value=1.0764}
          <at 0.7761 0.7175 RCIM_Cubic>
         ]}
        GreenCurve {keys [
          <at 0 2 RCIM_Cubic>
          {arriveTangent=-2.6857, interpMode=RCIM_Cubic, leaveTangent=-2.6857, time=0.4632, value=1.7812}
          {arriveTangent=-0.0141, interpMode=RCIM_Cubic, leaveTangent=-0.0141, tangentMode=RCTM_User, time=0.6477, value=0.2604}
          <at 0.7761 0 RCIM_Cubic>
         ]}
        RedCurve {keys [
          <at 0 0 RCIM_Cubic>
          {arriveTangent=1.0837, interpMode=RCIM_Cubic, leaveTangent=1.0837, tangentMode=RCTM_User, time=0.5867, value=0.1037}
          {arriveTangent=4.7319, interpMode=RCIM_Cubic, leaveTangent=4.7319, time=0.6815, value=3.9045}
          <at 0.7761 1 RCIM_Cubic>
         ]}
       }}))
    ScaleA=true
    ScaleRGB=true
   )
   &ScaleColor(
    Scale RGB=&VectorFromFloat(Value=&FloatFromCurve(
      FloatCurve=DataInterface() {propertyValues {Curve {keys [
          {arriveTangent=39.28, interpMode=RCIM_Cubic, leaveTangent=256.3728, tangentMode=RCTM_Break, time=0, value=1.5}
          <at 0.1658 7>
          {arriveTangent=-5.1937, interpMode=RCIM_Cubic, leaveTangent=-5.1937, tangentMode=RCTM_User, time=0.2567, value=1}
          <at 1 0.5 RCIM_Cubic>
         ]}}}
      Scale Curve=0.5
     ))
    ScaleA=true
    ScaleRGB=true
   ) {module=ScaleColor001}
   &ScaleColor(Scale RGB=&VectorFromFloat(Value=$System.TOD), ScaleA=true, ScaleRGB=true) {module=ScaleColor002}
  ]
  rendererClasses [&NiagaraSpriteRendererProperties]
  renderers [{rendererClass=&NiagaraSpriteRendererProperties}]
 }
 Glitch_Decals_Foot {
  emitterUpdateScript [
   &EmitterState(Life Cycle Mode=::NewEnumerator1, Loop Behavior=::NewEnumerator1, Loop Delay=0.127, Loop Duration=$User.WipeDuration, UseLoopDelay=true)
   &SpawnBurst_Instantaneous(Spawn Count=2, Use Spawn Probability=false, UseLoopCountLimit=false)
  ]
  particleSpawnScript [
   /SetVariables(Particles.Random=&RandomRangeFloat(Minimum=0, Maximum=1)) {isSetParameters=true}
   &FN_SetSpawnGroup(Number of Spawn Groups=2)
   /Niagara/Modules/Spawn/Initialization/InitializeParticle(
    Color=<rgba 2.2391 0 5 5>
    Lifetime=0.5
    Module.Write Color=true
    Module.Write Lifetime=true
    Module.Write Mass=true
    Module.Write Material Random=true
    Module.Write Position=false
    Module.Write Scale=false
    Module.Write SpriteRotation=false
    Module.Write SpriteSize=true
    Position=&SimulationPosition()
    Sprite Size=<30 30>
   )
   &CameraOffset(Camera Offset Amount=30)
  ]
  particleUpdateScript [
   &SkeletalMeshLocation(
    Bone / Socket Index=$Particles.SpawnGroup
    Bone Sampling Mode=::NewEnumerator4
    MaskSpawnGroup=false
    Skeletal Mesh=DataInterface() {propertyValues {
      DefaultMesh=null
      ExcludeBoneName=root
      OptionalComponentClass=null
      PreviewMesh=&F_MED_Veiled
      SkinningMode=SkinOnTheFly
      SoftSourceActor=null
      SourceMode=Default
      UvSetIndex=0
      WholeMeshLOD=-1
      bExcludeBone=false
      bReadDeformedGeometry=true
      bRequireCurrentFrameData=true
      ComponentTags []
      FilteredBones [foot_l,foot_r]
      FilteredSockets []
      MeshUserParameter {parameter {name=None, typeDefHandle {registeredTypeIndex=1}, varData []}}
      SamplingRegions []
     }}
   )
   &ParticleState()
   &SolveForcesAndVelocity(Clamp Velocity=false, Limit Acceleration=false)
   &DynamicMaterialParameters(
    Index 0 Param 1=&Lerp_Float(A=0.2, B=1.1, Alpha=$Particles.Random)
    Index 0 Param 2=&Lerp_Float(A=1, B=2, Alpha=$Particles.Random)
    Index 0 Param 3=&Lerp_Float(A=0.3, B=1, Alpha=$Particles.Random)
    Index 0 Param 4=-1
    Module.Param0WriteEnabled=true
    Module.Param1WriteEnabled=true
    Module.Param2WriteEnabled=true
    Module.Param3WriteEnabled=true
   )
  ]
  rendererClasses [&NiagaraSpriteRendererProperties]
  renderers [{rendererClass=&NiagaraSpriteRendererProperties}]
 }
 Glow {
  emitterUpdateScript [
   &EmitterState(Life Cycle Mode=::NewEnumerator1, Loop Behavior=::NewEnumerator1, Loop Delay=0.2, Loop Duration=$User.WipeDuration, UseLoopDelay=true)
   &SpawnBurst_Instantaneous(Use Spawn Probability=false, UseLoopCountLimit=false)
  ]
  particleSpawnScript [/Niagara/Modules/Spawn/Initialization/V2/InitializeParticle(
    Color=<rgba 1 1 1 0.5>
    Position Offset=<-10 7.5 0>
    Position Offset Coordinate Space=::Local
    Sprite Size=<120 190>
    Sprite Size Mode=::NewEnumerator4
    UsePositionOffset=true
    Write Material Random=true
   )]
  particleUpdateScript [
   &ParticleState()
   &CameraOffset()
   &Color(Color=&ColorFromCurve(
     ColorCurve=DataInterface() {propertyValues {
       AlphaCurve {keys [<at 0 0>,<at 0.1901 1 RCIM_Cubic>,<at 0.8095 1>,<at 1 0 RCIM_Cubic>]}
       BlueCurve {keys [<at 0 5 RCIM_Cubic>,<at 1 1 RCIM_Cubic>]}
       GreenCurve {keys [<at 0 5 RCIM_Cubic>,<at 1 1 RCIM_Cubic>]}
       RedCurve {keys [<at 0 5 RCIM_Cubic>,<at 1 1 RCIM_Cubic>]}
      }}
     CurveIndex=$System.NormalizedLoopAge
    ))
   &ScaleColor(
    Linear Color Curve=DataInterface() {propertyValues {
      AlphaCurve {keys [
        {arriveTangent=20.9953, interpMode=RCIM_Cubic, leaveTangent=20.9953, tangentMode=RCTM_User, time=0, value=0}
        {arriveTangent=3.1667, interpMode=RCIM_Cubic, leaveTangent=3.1667, time=0.0466, value=0.8633}
        {arriveTangent=-0.2888, interpMode=RCIM_Cubic, leaveTangent=-0.2888, time=0.3, value=0.95}
        {arriveTangent=-1.1599, interpMode=RCIM_Cubic, leaveTangent=-1.1599, time=0.5458, value=0.7191}
        {arriveTangent=-1.5832, interpMode=RCIM_Cubic, leaveTangent=-1.5832, time=0.9191, value=0.2319}
        <at 1 0 RCIM_Cubic>
       ]}
      BlueCurve {keys [{arriveTangent=-0.7585, interpMode=RCIM_Cubic, time=0.0026, value=1.5}, <at 1 1 RCIM_Cubic>]}
      GreenCurve {keys [<at 0.0026 0.6 RCIM_Cubic>,<at 1 0 RCIM_Cubic>]}
      RedCurve {keys [{arriveTangent=-1.8627, interpMode=RCIM_Cubic, time=0.0026, value=0}, <at 1 0.1807 RCIM_Cubic>]}
     }}
    Scale Mode=::NewEnumerator2
   )
  ]
  rendererClasses [&NiagaraSpriteRendererProperties]
  renderers [{rendererClass=&NiagaraSpriteRendererProperties}]
 }
 Wipe {
  emitterUpdateScript [
   &EmitterState(Life Cycle Mode=::NewEnumerator1, Loop Behavior=::NewEnumerator1, Loop Delay=0.2, Loop Duration=$User.WipeDuration, UseLoopDelay=true)
   &SpawnBurst_Instantaneous(Use Spawn Probability=false, UseLoopCountLimit=false)
  ]
  particleSpawnScript [&InitializeParticle(
    Lifetime=$User.WipeDuration
    Mesh Scale=<150 150 200>
    Mesh Scale Mode=::NewEnumerator4
    Position Offset Coordinate Space=::Local
    UsePositionOffset=true
    Write Material Random=true
   )]
  particleUpdateScript [
   &ParticleState()
   &CameraOffset(Camera Offset Amount=20)
   &Color(Color=&ColorFromCurve(
     ColorCurve=DataInterface() {propertyValues {
       AlphaCurve {keys [<at 0 0>,<at 0.1901 1 RCIM_Cubic>,<at 0.8095 1>,<at 1 0 RCIM_Cubic>]}
       BlueCurve {keys [<at 0 5 RCIM_Cubic>,<at 1 1 RCIM_Cubic>]}
       GreenCurve {keys [<at 0 5 RCIM_Cubic>,<at 1 1 RCIM_Cubic>]}
       RedCurve {keys [<at 0 5 RCIM_Cubic>,<at 1 1 RCIM_Cubic>]}
      }}
     CurveIndex=$System.NormalizedLoopAge
    ))
   &ScaleColor(Scale Mode=::NewEnumerator2)
  ]
  rendererClasses [&NiagaraMeshRendererProperties]
  renderers [{rendererClass=&NiagaraMeshRendererProperties, rendererIndex=0}]
 }
 Wipe001 {
  emitterUpdateScript [
   &EmitterState(Life Cycle Mode=::NewEnumerator1, Loop Behavior=::NewEnumerator1, Loop Delay=0.2, Loop Duration=$User.WipeDuration, UseLoopDelay=true)
   &SpawnBurst_Instantaneous(Use Spawn Probability=false, UseLoopCountLimit=false)
  ]
  particleSpawnScript [&InitializeParticle(
    Lifetime=$User.WipeDuration
    Position Offset Coordinate Space=::Local
    Sprite Size=<90 170>
    Sprite Size Mode=::NewEnumerator4
    UsePositionOffset=true
    Write Material Random=true
   )]
  particleUpdateScript [
   &ParticleState()
   &CameraOffset(Camera Offset Amount=20)
   &Color(Color=&ColorFromCurve(
     ColorCurve=DataInterface() {propertyValues {
       AlphaCurve {keys [<at 0 0>,<at 0.1901 1 RCIM_Cubic>,<at 0.8095 1>,<at 1 0 RCIM_Cubic>]}
       BlueCurve {keys [<at 0 5 RCIM_Cubic>,<at 1 1 RCIM_Cubic>]}
       GreenCurve {keys [<at 0 5 RCIM_Cubic>,<at 1 1 RCIM_Cubic>]}
       RedCurve {keys [<at 0 5 RCIM_Cubic>,<at 1 1 RCIM_Cubic>]}
      }}
     CurveIndex=$System.NormalizedLoopAge
    ))
   &ScaleColor(Scale Mode=::NewEnumerator2)
  ]
  rendererClasses [&NiagaraSpriteRendererProperties]
  renderers [{rendererClass=&NiagaraSpriteRendererProperties}]
 }
 Wipe002 {
  emitterUpdateScript [
   &EmitterState(Life Cycle Mode=::NewEnumerator1, Loop Behavior=::NewEnumerator1, Loop Delay=0.127, Loop Duration=$User.WipeDuration, UseLoopDelay=true)
   &SpawnBurst_Instantaneous(Use Spawn Probability=false, UseLoopCountLimit=false)
  ]
  particleSpawnScript [&InitializeParticle(
    Lifetime=0.75
    Mesh Scale=<150 150 200>
    Mesh Scale Mode=::NewEnumerator4
    Position Offset Coordinate Space=::Local
    UsePositionOffset=true
    Write Material Random=true
   )]
  particleUpdateScript [
   &ParticleState()
   &CameraOffset(Camera Offset Amount=20)
   &Color(Color=&ColorFromCurve(
     ColorCurve=DataInterface() {propertyValues {
       AlphaCurve {keys [<at 0 0>,<at 0.1901 1 RCIM_Cubic>,<at 0.8095 1>,<at 1 0 RCIM_Cubic>]}
       BlueCurve {keys [<at 0 5 RCIM_Cubic>,<at 1 1 RCIM_Cubic>]}
       GreenCurve {keys [<at 0 5 RCIM_Cubic>,<at 1 1 RCIM_Cubic>]}
       RedCurve {keys [<at 0 5 RCIM_Cubic>,<at 1 1 RCIM_Cubic>]}
      }}
     CurveIndex=$System.NormalizedLoopAge
    ))
   &ScaleColor(Scale Mode=::NewEnumerator2)
  ]
  rendererClasses [&NiagaraMeshRendererProperties]
  renderers [{rendererClass=&NiagaraMeshRendererProperties, rendererIndex=0}]
 }
 Wipe003 {
  emitterUpdateScript [
   &EmitterState(Life Cycle Mode=::NewEnumerator1, Loop Behavior=::NewEnumerator1, Loop Delay=0.127, Loop Duration=$User.WipeDuration, UseLoopDelay=true)
   &SpawnBurst_Instantaneous(Use Spawn Probability=false, UseLoopCountLimit=false)
  ]
  particleSpawnScript [&InitializeParticle(
    Lifetime=0.75
    Position Offset Coordinate Space=::Local
    Sprite Size=<100 185>
    Sprite Size Mode=::NewEnumerator4
    UsePositionOffset=true
    Write Material Random=true
   )]
  particleUpdateScript [
   &ParticleState()
   &CameraOffset(Camera Offset Amount=20)
   &Color(Color=&ColorFromCurve(
     ColorCurve=DataInterface() {propertyValues {
       AlphaCurve {keys [<at 0 0>,<at 0.1901 1 RCIM_Cubic>,<at 0.8095 1>,<at 1 0 RCIM_Cubic>]}
       BlueCurve {keys [<at 0 5 RCIM_Cubic>,<at 1 1 RCIM_Cubic>]}
       GreenCurve {keys [<at 0 5 RCIM_Cubic>,<at 1 1 RCIM_Cubic>]}
       RedCurve {keys [<at 0 5 RCIM_Cubic>,<at 1 1 RCIM_Cubic>]}
      }}
     CurveIndex=$System.NormalizedLoopAge
    ))
   &ScaleColor(Scale Mode=::NewEnumerator2)
  ]
  rendererClasses [&NiagaraSpriteRendererProperties]
  renderers [{rendererClass=&NiagaraSpriteRendererProperties}]
 }
}
summary {
 systemName=NS_Veiled_Swap_Emote_GlitchToDefault
 emitters [
  {emitterName=Wipe, rendererClasses [&NiagaraMeshRendererProperties]}
  {emitterName=Cube_Glitch_Head001, rendererClasses [&NiagaraMeshRendererProperties]}
  {emitterName=Cube_Glitch_Head, rendererClasses [&NiagaraSpriteRendererProperties]}
  {emitterName=Cube_Glitch_Head002, rendererClasses [&NiagaraMeshRendererProperties]}
  {emitterName=Wipe001, rendererClasses [&NiagaraSpriteRendererProperties]}
  {emitterName=Fresnel_Glow_Swipe, rendererClasses [&NiagaraSpriteRendererProperties]}
  {emitterName=Wipe002, rendererClasses [&NiagaraMeshRendererProperties]}
  {emitterName=Wipe003, rendererClasses [&NiagaraSpriteRendererProperties]}
  {emitterName=Cube_Glitch_Head003, rendererClasses [&NiagaraSpriteRendererProperties]}
  {emitterName=Glitch_Decals_Foot, rendererClasses [&NiagaraSpriteRendererProperties]}
  {emitterName=Glow, rendererClasses [&NiagaraSpriteRendererProperties]}
 ]
 userVariables [
  User.bIsFrontEnd {type=NiagaraBool}
  User.bIsFrontEndPreview {type=NiagaraBool}
  User.SkeletalMesh {
   type=NiagaraDataInterfaceSkeletalMesh
   defaultValue {
    struct=NiagaraExt_VariableValue_DataInterface
    value {dataInterface=&self:NiagaraDataInterfaceSkeletalMesh_1}
   }
  }
  User.WipeDuration {type=NiagaraFloat, defaultValue=0.8, description=""}
 ]
}
